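# Level Sequence for the PrawnTwine M menu idle face animation (Face1), apparently an authoring scene for the cosmetic skin.
# A spawnable FortPlayerMannequin at the origin plays a skeletal animation section over frames 0-300 at 30 fps.
# A Fortnite_Legacy_Face_CtrlRig track is layered on it at full weight.
# Its visible channels only hold the master face control's placement and scale, with head-follow on and flipbook off, showing no animated facial poses.
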LevelSequence /BRCosmetics/Animation/Game/MainPlayer/Menu/BR/CustomSkins/_s39/PrawnTwineM/PT_anim/PrawnTwine_M_idle_Face1
bindings [FortPlayerMannequin {
  id=#1
  possessed_class=FortPlayerMannequin
  tracks [
   {type=MovieSceneSpawnTrack, sections [{type=MovieSceneSpawnSection, channels [None {type=MovieSceneScriptingBoolChannel, default=true}]}]}
   {
    type=MovieScene3DTransformTrack
    sections [{
      type=MovieScene3DTransformSection
      channels [
       Location.X {type=MovieSceneScriptingDoubleChannel, default=0}
       Location.Y {type=MovieSceneScriptingDoubleChannel, default=0}
       Location.Z {type=MovieSceneScriptingDoubleChannel, default=0}
       Rotation.X {type=MovieSceneScriptingDoubleChannel, default=0}
       Rotation.Y {type=MovieSceneScriptingDoubleChannel, default=0}
       Rotation.Z {type=MovieSceneScriptingDoubleChannel, default=0}
       Scale.X {type=MovieSceneScriptingDoubleChannel, default=1}
       Scale.Y {type=MovieSceneScriptingDoubleChannel, default=1}
       Scale.Z {type=MovieSceneScriptingDoubleChannel, default=1}
      ]
     }]
   }
   {type=MovieSceneSkeletalAnimationTrack, sections [{type=MovieSceneSkeletalAnimationSection, end_frame=300, start_frame=0}]}
   {
    type=MovieSceneControlRigParameterTrack
    display_name=Fortnite_Legacy_Face_CtrlRig
    sections [{
      type=MovieSceneControlRigParameterSection
      channels [
       Flipbook {type=MovieSceneScriptingBoolChannel, default=false, const [false,7]}
       Face_Ctrl_Master.Location.X {type=MovieSceneScriptingFloatChannel, default=-33.1501, const [-38.9801,8]}
       Face_Ctrl_Master.Location.Y {type=MovieSceneScriptingFloatChannel, default=-6.7891, const [-7.4023,8]}
       Face_Ctrl_Master.Location.Z {type=MovieSceneScriptingFloatChannel, default=25.9951, const [1.4556,8]}
       Face_Ctrl_Master.Rotation.X {type=MovieSceneScriptingFloatChannel, default=0, const [0,7]}
       Face_Ctrl_Master.Rotation.Y {type=MovieSceneScriptingFloatChannel, default=0, const [0,7]}
       Face_Ctrl_Master.Rotation.Z {type=MovieSceneScriptingFloatChannel, default=0, const [0,7]}
       Face_Ctrl_Master.Scale.X {type=MovieSceneScriptingFloatChannel, default=2.5, const [2.5,7]}
       Face_Ctrl_Master.Scale.Y {type=MovieSceneScriptingFloatChannel, default=2.5, const [2.5,7]}
       Face_Ctrl_Master.Scale.Z {type=MovieSceneScriptingFloatChannel, default=2.5, const [2.5,7]}
       gui_follow_head {type=MovieSceneScriptingFloatChannel, default=1, const [1,7]}
       Weight {type=MovieSceneScriptingFloatChannel, default=1}
      ]
     }]
   }
  ]
 }]
display_rate {denominator=1, numerator=30}
tick_resolution {denominator=1, numerator=24000}
tracks []
